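# Physics asset for the EvilSuit pickaxe cosmetic skeletal mesh, a two-handed melee weapon.
# It holds a single capsule body on the Melee_TwoHanded_Handle bone, a wide shape offset along the handle that approximates the whole weapon.
# There are no constraints, so the pickaxe has one rigid collision shape with no jointed or simulated sub-parts.
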PhysicsAsset /BRCosmetics/Weapons/FORT_Melee/Pickaxe_EvilSuit/Meshes/SK_Pickaxe_EvilSuit_PhysicsAsset
bodies {Melee_TwoHanded_Handle [{shapeType=Capsule, length=70.7834, radius=42.0182, center=<0.71 -0.3532 26.6453>, rotation=<rot -3.4813 -90 0.0202>}]}
constraints []
